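; Blimp spotlight Blueprint that supplies per-instance beam settings to a separate spotlight FX actor.
; On server setup it runs the parent logic, then, with authority only, casts the FX actor to BP_Blimp_SpotLight_01_FXActor.
; It packs its color, beam size and rotation settings into a spotlight data struct and sets that on the FX actor.
; The editor-only spawn override is an empty stub, and the construction script only calls the parent.
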(fn editorOnly_SpawnFXActor
  (params (actorClass :type Actor%20Class%20Reference) (fxComp :type Persistent%20FXComponent%20Object%20Reference))
  (return 0))

(fn setupFxActorOnServer
  (params (meshComp :type Static%20Mesh%20Component%20Object%20Reference) (FXActor :type Actor%20Object%20Reference))
  (|CallParentFunction meshComp FXActor)
  (Networking|SwitchHasAuthority
    (:Authority
      (bind _asbp_blimp_spot_light_01_fxactor (Utilities|Casting|CastToBP_Blimp_SpotLight_01_FXActor FXActor))
      (Variables|/Game/Effects/Fort_Effects/BluePrints/Environments/Blimp/BP_Blimp_SpotLight_01_FXActor.BP_Blimp_SpotLight_01_FXActor_C|SetspotlightData (Utilities|Struct|MakeSpotlightDataStruct (Variables|SpotlightFXSettings|GetSpotlightColor) (Variables|SpotlightFXSettings|GetBeam%20Height) (Variables|SpotlightFXSettings|GetBeam%20Width) (Variables|SpotlightFXSettings|GetRotation%20Max) (Variables|SpotlightFXSettings|GetRotation%20Speed)) _asbp_blimp_spot_light_01_fxactor))))

(fn ConstructionScript
  (|CallParentFunction))
