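; Blimp spotlight FX actor that drives a tilted light-shaft mesh as a configurable rotating beam.
; A spotlight data struct supplies color, beam height, beam width, rotation max and rotation speed, which are written straight to the mesh's material parameters.
; The construction script applies them unconditionally, and the data's OnRep reapplies them only where cosmetic events can execute.
; The editor-only FX-actor spawn function is a stub that returns an empty value.
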(fn editorOnly_SpawnFXActor
  (params (actorClass :type Actor%20Class%20Reference) (fxComp :type Persistent%20FXComponent%20Object%20Reference))
  (return 0))

(fn OnRep_spotlightData
  (bind _spotlightdata (Variables|Default|GetspotlightData))
  (bind (_spotlightcolor_2_5609fd3e4239b11ae09e748bb54d70b5 _beamheight_5_51e8372e4a786ee9cd153184f17ebf72 _beamwidth_7_126606a54154f953d7e7f1a1373d889e _rotationmax_13_233bff9a43a3754e323416847188d1c1 _rotationspeed_15_305f536c44abcfbac3d8ceb32c451a01) (Utilities|Struct|BreakSpotlightDataStruct _spotlightdata))
  (Networking|CanExecuteCosmeticEvents
    (:True
      (Rendering|Material|SetScalarParameterValueonMaterials (Variables|Default|GetEditorOnly_sm_neo_tilted_light_shaft) "Rotation Max" _rotationmax_13_233bff9a43a3754e323416847188d1c1)
      (Rendering|Material|SetScalarParameterValueonMaterials (Variables|Default|GetEditorOnly_sm_neo_tilted_light_shaft) "Rotation Speed" _rotationspeed_15_305f536c44abcfbac3d8ceb32c451a01)
      (Rendering|Material|SetScalarParameterValueonMaterials (Variables|Default|GetEditorOnly_sm_neo_tilted_light_shaft) "ScaleHeight" _beamheight_5_51e8372e4a786ee9cd153184f17ebf72)
      (Rendering|Material|SetVectorParameterValueonMaterials (Variables|Default|GetEditorOnly_sm_neo_tilted_light_shaft) "Color" (Math|Conversions|ToVector(LinearColor) _spotlightcolor_2_5609fd3e4239b11ae09e748bb54d70b5))
      (Rendering|Material|SetScalarParameterValueonMaterials (Variables|Default|GetEditorOnly_sm_neo_tilted_light_shaft) "MainWidth" _beamwidth_7_126606a54154f953d7e7f1a1373d889e))))

(fn ConstructionScript
  (bind _spotlightdata (Variables|Default|GetspotlightData))
  (bind (_spotlightcolor_2_5609fd3e4239b11ae09e748bb54d70b5 _beamheight_5_51e8372e4a786ee9cd153184f17ebf72 _beamwidth_7_126606a54154f953d7e7f1a1373d889e _rotationmax_13_233bff9a43a3754e323416847188d1c1 _rotationspeed_15_305f536c44abcfbac3d8ceb32c451a01) (Utilities|Struct|BreakSpotlightDataStruct _spotlightdata))
  (|CallParentFunction)
  (Rendering|Material|SetScalarParameterValueonMaterials (Variables|Default|GetEditorOnly_sm_neo_tilted_light_shaft) "Rotation Max" _rotationmax_13_233bff9a43a3754e323416847188d1c1)
  (Rendering|Material|SetScalarParameterValueonMaterials (Variables|Default|GetEditorOnly_sm_neo_tilted_light_shaft) "Rotation Speed" _rotationspeed_15_305f536c44abcfbac3d8ceb32c451a01)
  (Rendering|Material|SetScalarParameterValueonMaterials (Variables|Default|GetEditorOnly_sm_neo_tilted_light_shaft) "ScaleHeight" _beamheight_5_51e8372e4a786ee9cd153184f17ebf72)
  (Rendering|Material|SetVectorParameterValueonMaterials (Variables|Default|GetEditorOnly_sm_neo_tilted_light_shaft) "Color" (Math|Conversions|ToVector(LinearColor) _spotlightcolor_2_5609fd3e4239b11ae09e748bb54d70b5))
  (Rendering|Material|SetScalarParameterValueonMaterials (Variables|Default|GetEditorOnly_sm_neo_tilted_light_shaft) "MainWidth" _beamwidth_7_126606a54154f953d7e7f1a1373d889e))
